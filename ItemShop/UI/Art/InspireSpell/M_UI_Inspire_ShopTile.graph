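# UI-domain material for an Inspire item shop tile, drawing animated stars over red, cream and dark grunge backgrounds.
# It holds several separate compositions, with no connection to a material output shown.
# Star masks come from counter-panning lineup textures, sine-timed sliding star rows, and a tiled SDF star pattern with random scale and rotation, scrolling oppositely per half.
# These masks blend colours built from grunge textures and linear gradients.
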Material /ItemShop/UI/Art/InspireSpell/M_UI_Inspire_ShopTile
graph {
 properties {bUsedWithStaticMesh=false, materialDomain=MD_UI, pixelDepthOffsetMode=PDOM_Legacy, usedShadingModels=Default Lit}
 nodes [
  n1 = Multiply(A=Sine(None=Multiply() {constB=2})) {constB=9}
  n2 = Multiply(A=Saturate(None=@n1)) {constB=8}
  n3 = MaterialFunctionCall!(
   Texture Scale=Constant3Vector() {constant=<rgba 0.2552 0.2552 0.2552 0>}
   UVs=MaterialFunctionCall() {materialFunction=&MF_UI_UVScaling}.Scaled To Fill
   Offset XY Texturesize Z=Constant3Vector() {constant=<rgba 0.03 0.03 0.0642 0>}
  ) {materialFunction=/Engine/Functions/Engine_MaterialFunctions01/Texturing/ScaleUVsAroundPoint}
  Add(A=AppendVector(A=@n2, B=Constant() {r=-1}), B=@n3.UVs)
  n4 = Constant() {r=2}
  n5 = MaterialFunctionCall() {materialFunction=&MF_UI_UVScaling}
  n6 = Constant3Vector() {constant=<rgba -0.053 0.0605 0.0642 0>}
  n7 = Multiply(A=Time()) {constB=0.25}
  n8 = TextureSample(UVs=Panner(
    Coordinate=MaterialFunctionCall(Texture Scale=@n4, UVs=@n5.ScaleY, Offset XY Texturesize Z=@n6) {materialFunction=&ScaleUVsAroundPoint}.UVs
    Time=@n7
   ) {speedX=0.1}) {samplerType=SAMPLERTYPE_LinearColor, texture=&/T_UI_InspireStarLineup}
  n9 = Constant3Vector() {constant=<rgba -0.013 0.009 0.0642 0>}
  n10 = TextureSample(UVs=Panner(
    Coordinate=MaterialFunctionCall(Texture Scale=@n4, UVs=@n5.ScaleY, Offset XY Texturesize Z=@n9) {materialFunction=&ScaleUVsAroundPoint}.UVs
    Time=@n7
   ) {speedX=-0.1}) {samplerType=SAMPLERTYPE_LinearColor, texture=&T_UI_InspireStarLineup}
  Add(A=SmoothStep(Value=@n8.R) {constMax=0.25}, B=SmoothStep(Value=@n10.R) {constMax=0.25})
  n11 = Constant() {r=-0.007}
  AppendVector(A=Constant(), B=LinearInterpolate(A=@n11, B=Multiply(A=@n11) {constB=-1}, Alpha=NamedRerouteUsage()) {constA=0.2, constB=-0.2})
  Constant3Vector() {constant=<rgba 0.026 0.0173 0.0066 0>}
  Constant3Vector() {constant=<rgba 0.026 0.0173 0.0066 0>}
  Constant3Vector() {constant=<rgba -0.021 -0.005 0.0642 0>}
  Constant3Vector() {constant=<rgba -0.036 -0.005 0.0642 0>}
  Constant3Vector() {constant=<rgba -0.05 -0.005 0.0642 0>}
  Constant3Vector() {constant=<rgba -0.065 -0.005 0.0642 0>}
  Constant3Vector() {constant=<rgba -0.08 -0.005 0.0642 0>}
  Constant3Vector() {constant=<rgba 0.085 0.04 0.0642 0>}
  Constant3Vector() {constant=<rgba 0.1 0.04 0.0642 0>}
  Constant3Vector() {constant=<rgba 0.115 0.04 0.0642 0>}
  Constant3Vector() {constant=<rgba 0.13 0.04 0.0642 0>}
  Constant3Vector() {constant=<rgba 0.145 0.04 0.0642 0>}
  Constant() {r=0.15}
  Constant() {r=0.15}
  Constant() {r=0.15}
  Constant() {r=0.15}
  Constant() {r=0.15}
  Constant() {r=0.15}
  Constant() {r=0.15}
  Constant() {r=0.15}
  Constant() {r=0.15}
  Constant() {r=0.15}
  Frac(None=@n1)
  n12 = Constant3Vector() {constant=<rgba 0.349 0.3121 0.271 0>}
  n13 = Constant3Vector() {constant=<rgba 0.7157 0.6654 0.5906 0>}
  n14 = TextureSample(UVs=MaterialFunctionCall() {materialFunction=&MF_UI_UVScaling}.Scaled To Fill) {texture=/OfferCatalog/Art/InspireSpell/T_UI_LightGrunge}
  n15 = TextureSample(UVs=MaterialFunctionCall(Texture Scale=Constant() {r=0.4987}, Offset XY Texturesize Z=Constant3Vector() {constant=<rgba 0.1 -0.01 0.0642 0>}) {materialFunction=&ScaleUVsAroundPoint}.UVs) {texture=&/T_UI_GraphicGrunge}
  n16 = MaterialFunctionCall() {materialFunction=&MF_UI_UVScaling}
  n17 = TextureSample(UVs=MaterialFunctionCall(Texture Scale=Constant() {r=0.4987}, UVs=@n16.Scaled to Fit, Offset XY Texturesize Z=Constant3Vector() {constant=<rgba -0.04 0.05 0.0642 0>}) {materialFunction=&ScaleUVsAroundPoint}.UVs) {texture=&T_UI_GraphicGrunge}
  n18 = Constant() {r=2}
  n19 = Panner(Coordinate=MaterialFunctionCall(Texture Scale=@n18, UVs=@n16.Scaled to Fit, Offset XY Texturesize Z=Constant3Vector() {constant=<rgba -0.013 0.018 0.0642 0>}) {materialFunction=&ScaleUVsAroundPoint}.UVs) {speedX=-0.1}
  n20 = Panner(Coordinate=MaterialFunctionCall(Texture Scale=@n18, UVs=@n16.Scaled to Fit, Offset XY Texturesize Z=Constant3Vector() {constant=<rgba -0.053 0.06 0.0642 0>}) {materialFunction=&ScaleUVsAroundPoint}.UVs) {speedX=0.1}
  LinearInterpolate(
   A=Constant3Vector() {constant=<rgba 0.599 0 0 0>}
   B=LinearInterpolate(B=LinearInterpolate(A=@n12, B=@n13, Alpha=@n14.R), Alpha=Multiply(A=@n15.R, B=@n17.R))
   Alpha=Multiply(
    A=TextureSample(UVs=@n19) {texture=&T_UI_InspireStarLineup}.R
    B=TextureSample(UVs=@n20) {texture=&T_UI_InspireStarLineup}.R
   )
  )
  LinearInterpolate(
   A=Constant3Vector() {constant=<rgba 0.1022 0.0823 0.0578 0>}
   B=Constant3Vector() {constant=<rgba 0.2789 0.2384 0.1946 1>}
   Alpha=TextureSample(UVs=MaterialFunctionCall() {materialFunction=&MF_UI_UVScaling}.Scaled To Fill) {texture=&T_UI_LightGrunge}.R
  )
  n21 = MaterialFunctionCall() {materialFunction=&MF_UI_UVScaling}
  n22 = MaterialFunctionCall(GradientStart=Constant() {r=-0.15}, GradientEnd=Constant() {r=0.5}, RotationAngle=Constant() {r=0.65}) {materialFunction=&MF_UI_LinearGradient}
  n23 = MaterialFunctionCall(GradientStart=Constant() {r=-0.05}, GradientEnd=Constant() {r=0.5}, RotationAngle=Constant() {r=0.15}) {materialFunction=&MF_UI_LinearGradient}
  n24 = LinearInterpolate(
   A=Constant3Vector() {constant=<rgba 0.0573 0.0503 0.0373 0>}
   B=LinearInterpolate(
    A=Constant3Vector() {constant=<rgba 0.3073 0.2721 0.2065 0>}
    B=Constant3Vector() {constant=<rgba 0.5573 0.5017 0.4412 1>}
    Alpha=TextureSample(UVs=@n21.Scaled To Fill) {texture=&T_UI_LightGrunge}.R
   )
   Alpha=Multiply(A=Multiply(A=@n22.Result, B=@n23.Result)) {constB=2}
  )
  LinearInterpolate(A=Constant3Vector() {constant=<rgba 0.026 0.026 0.026 0>}, B=@n24)
  n25 = Constant3Vector() {constant=<rgba 0.3073 0.2721 0.2065 0>}
  n26 = Constant3Vector() {constant=<rgba 0.5573 0.5017 0.4412 1>}
  n27 = TextureSample(UVs=MaterialFunctionCall() {materialFunction=&MF_UI_UVScaling}.Scaled To Fill) {texture=&T_UI_LightGrunge}
  n28 = Multiply(
   A=MaterialFunctionCall(GradientStart=Constant() {r=-0.15}, GradientEnd=Constant() {r=0.5}, RotationAngle=Constant() {r=0.65}) {materialFunction=&MF_UI_LinearGradient}.Result
   B=MaterialFunctionCall(GradientStart=Constant() {r=-0.05}, GradientEnd=Constant() {r=0.5}, RotationAngle=Constant() {r=0.15}) {materialFunction=&MF_UI_LinearGradient}.Result
  )
  n29 = MaterialFunctionCall(UVs=TextureCoordinate(), Texture Scale=Constant() {r=2}) {materialFunction=&ScaleUVsByCenter}
  n30 = Constant() {desc=pan speed, r=0.04}
  n31 = Round(None=ComponentMask(None=TextureCoordinate()) {g=true})
  n32 = AppendVector(
   A=LinearInterpolate(A=@n30, B=Multiply(A=@n30) {constB=-1}, Alpha=NamedRerouteDeclaration(None=@n31) {name=up and down mask, nodeColor=<rgba 1 0.1641 0 1>}) {constA=-0.1, constB=0.1}
   B=Constant()
  )
  n33 = Constant()
  n34 = Constant() {r=-0.007}
  n35 = LinearInterpolate(A=@n34, B=Multiply(A=@n34) {constB=-16}, Alpha=NamedRerouteUsage()) {constA=0.2, constB=-0.2}
  n36 = Frac(None=Add(A=Panner(Coordinate=@n29.UVs, Speed=@n32), B=AppendVector(A=@n33, B=@n35)))
  n37 = Constant() {desc=num stars in a row, r=9}
  n38 = MaterialFunctionCall!(UV=@n36, Tiling X=@n37, Tiling Y=Constant() {r=2}) {materialFunction=/FortUILibrary/MaterialFunctions/Basics/MF_UI_Pattern}
  n39 = Multiply(A=ComponentMask(None=ComponentMask(None=@n36) {r=true}) {r=true}, B=@n37)
  n40 = MaterialFunctionCall!(UV=Divide(A=Truncate!(None=@n39) {is=class MaterialExpressionTruncate}, B=Constant() {r=4}), Seed=Constant() {r=0.1}) {materialFunction=&MF_UI_Random}
  n41 = LinearInterpolate(Alpha=@n40.Result) {constA=0.3, constB=1.1}
  n42 = Frac(None=Add(A=@n40.Result) {constB=0.123})
  n43 = TextureSample(UVs=MaterialFunctionCall(
    UVs=MaterialFunctionCall(UVs=@n38.Scaled Pattern UVs, Texture Scale=@n41) {materialFunction=&ScaleUVsByCenter}.UVs
    "Rotation Angle (0-1)"=LinearInterpolate(Alpha=@n42) {constB=0.03}
   ) {materialFunction=&CustomRotator}.Rotated Values) {samplerSource=SSM_Clamp_WorldGroupSettings, samplerType=SAMPLERTYPE_Alpha, texture=/Game/UI/Frontend/ItemShop/Art/InspireSpell/T_UI_SDFstar}
  LinearInterpolate(
   A=LinearInterpolate(
    A=Constant3Vector() {constant=<rgba 0.0573 0.0503 0.0373 0>}
    B=LinearInterpolate(A=@n25, B=@n26, Alpha=@n27.R)
    Alpha=Multiply(A=@n28) {constB=2}
   )
   B=Constant3Vector() {constant=<rgba 0.0087 0.0087 0.0087 0>}
   Alpha=SmoothStep(Value=@n43.R) {constMax=0.5, constMin=0.4}
  )
  n44 = MaterialFunctionCall(Texture Scale=Constant() {r=0.4987}, UVs=@n5.ScaleY, Offset XY Texturesize Z=Constant3Vector() {constant=<rgba 0.05 0.01 0.0642 0>}) {materialFunction=&ScaleUVsAroundPoint}
  n45 = MaterialFunctionCall(Texture Scale=Constant() {r=0.4987}, UVs=@n5.ScaleY, Offset XY Texturesize Z=Constant3Vector() {constant=<rgba -0.05 0.05 0.0642 0>}) {materialFunction=&ScaleUVsAroundPoint}
  LinearInterpolate(
   A=Constant3Vector() {constant=<rgba 0.3333 0.0059 0.0005 0>}
   B=@n24
   Alpha=Multiply(
    A=TextureSample(UVs=@n44.UVs) {texture=&T_UI_GraphicGrunge}.R
    B=TextureSample(UVs=@n45.UVs) {texture=&T_UI_GraphicGrunge}.R
   )
  )
  n46 = TextureSample() {texture=&T_UI_InspireStarLineup}
  n47 = Constant3Vector() {constant=<rgba 0.9497 0.0226 0.0131 0>}
  n48 = Sine(None=Multiply(A=Time()) {constB=0.25}) {period=1.5}
  n49 = Saturate(None=Add(A=Multiply(A=@n48) {constB=3}) {constB=-1})
  n50 = Multiply(
   A=AppendVector(A=Multiply(A=@n49) {constB=10}, B=Constant() {r=-1})
   B=MaterialFunctionCall(GradientStart=Constant(), GradientEnd=Constant() {r=1}) {materialFunction=&MF_UI_LinearGradient}.Result
  )
  LinearInterpolate(A=LinearInterpolate(A=Add(A=@n46.R, B=@n47), B=@n46.R, Alpha=Saturate(None=@n50)))
  n51 = Multiply(A=MaterialFunctionCall() {materialFunction=&MF_UI_UVScaling}.Scaled To Fill) {constB=2}
  n52 = Constant() {r=0.1}
  n53 = MaterialFunctionCall(
   Texture Scale=@n52
   UVs=MaterialFunctionCall() {materialFunction=&MF_UI_UVScaling}.ScaleY
   Offset XY Texturesize Z=Constant3Vector() {constant=<rgba -0.019 -0.005 0.0642 0>}
  ) {materialFunction=&ScaleUVsAroundPoint}
  n54 = Multiply(A=Time()) {constB=0.15}
  n55 = Add(A=Multiply(A=Sine(None=@n54) {period=1.5}) {constB=3}) {constB=-1.5}
  n56 = AppendVector(A=Multiply(A=Saturate(None=@n55)) {constB=-7}, B=Constant() {r=-1})
  n57 = MaterialFunctionCall(
   Texture Scale=@n52
   UVs=MaterialFunctionCall() {materialFunction=&MF_UI_UVScaling}.ScaleY
   Offset XY Texturesize Z=Constant3Vector() {constant=<rgba -0.03 -0.005 0.0642 0>}
  ) {materialFunction=&ScaleUVsAroundPoint}
  n58 = Multiply(A=Time()) {constB=0.15}
  n59 = Add(A=Multiply(A=Sine(None=@n58) {period=1.5}) {constB=3}) {constB=-1.5}
  n60 = AppendVector(A=Multiply(A=Saturate(None=@n59)) {constB=-7}, B=Constant() {r=-1})
  n61 = Multiply(
   A=TextureSample(UVs=Add(A=@n53.UVs, B=@n56)) {texture=&/T_UI_InspireStar}.R
   B=TextureSample(UVs=Add(A=@n57.UVs, B=@n60)) {texture=&T_UI_InspireStar}.R
  )
  n62 = MaterialFunctionCall(
   Texture Scale=@n52
   UVs=MaterialFunctionCall() {materialFunction=&MF_UI_UVScaling}.ScaleY
   Offset XY Texturesize Z=Constant3Vector() {constant=<rgba -0.063 -0.005 0.0642 0>}
  ) {materialFunction=&ScaleUVsAroundPoint}
  n63 = Time()
  n64 = Multiply(A=@n63) {constB=0.15}
  n65 = Add(A=Multiply(A=Sine(None=@n64) {period=1.5}) {constB=3}) {constB=-1.5}
  n66 = AppendVector(A=Multiply(A=Saturate(None=@n65)) {constB=-7}, B=Constant() {r=-1})
  n67 = Time()
  n68 = Sine(None=Multiply(A=@n67) {constB=0.15}) {period=1.5}
  n69 = Saturate(None=Add(A=Multiply(A=@n68) {constB=3}) {constB=-1.5})
  n70 = Add(
   A=MaterialFunctionCall(
    Texture Scale=@n52
    UVs=MaterialFunctionCall() {materialFunction=&MF_UI_UVScaling}.ScaleY
    Offset XY Texturesize Z=Constant3Vector() {constant=<rgba -0.042 -0.005 0.0642 0>}
   ) {materialFunction=&ScaleUVsAroundPoint}.UVs
   B=AppendVector(A=Multiply(A=@n69) {constB=-7}, B=Constant() {r=-1})
  )
  n71 = Time()
  n72 = Sine(None=Multiply(A=@n71) {constB=0.15}) {period=1.5}
  n73 = Saturate(None=Add(A=Multiply(A=@n72) {constB=3}) {constB=-1.5})
  n74 = Add(
   A=MaterialFunctionCall(
    Texture Scale=@n52
    UVs=MaterialFunctionCall() {materialFunction=&MF_UI_UVScaling}.ScaleY
    Offset XY Texturesize Z=Constant3Vector() {constant=<rgba -0.053 -0.005 0.0642 0>}
   ) {materialFunction=&ScaleUVsAroundPoint}.UVs
   B=AppendVector(A=Multiply(A=@n73) {constB=-7}, B=Constant() {r=-1})
  )
  n75 = Multiply(
   A=TextureSample(UVs=Add(A=@n62.UVs, B=@n66)) {texture=&T_UI_InspireStar}.R
   B=Multiply(
    A=TextureSample(UVs=@n70) {texture=&T_UI_InspireStar}.R
    B=TextureSample(UVs=@n74) {texture=&T_UI_InspireStar}.R
   )
  )
  n76 = MaterialFunctionCall(
   Texture Scale=@n52
   UVs=MaterialFunctionCall() {materialFunction=&MF_UI_UVScaling}.ScaleY
   Offset XY Texturesize Z=Constant3Vector() {constant=<rgba 0.085 0.04 0.0642 0>}
  ) {materialFunction=&ScaleUVsAroundPoint}
  n77 = Time()
  n78 = Multiply(A=@n77) {constB=0.15}
  n79 = Add(A=Multiply(A=Sine(None=@n78) {period=1.5}) {constB=3}) {constB=-1.5}
  n80 = AppendVector(A=Multiply(A=Saturate(None=@n79)) {constB=7}, B=Constant() {r=-1})
  n81 = MaterialFunctionCall(
   Texture Scale=@n52
   UVs=MaterialFunctionCall() {materialFunction=&MF_UI_UVScaling}.ScaleY
   Offset XY Texturesize Z=Constant3Vector() {constant=<rgba 0.097 0.04 0.0642 0>}
  ) {materialFunction=&ScaleUVsAroundPoint}
  n82 = Time()
  n83 = Multiply(A=@n82) {constB=0.15}
  n84 = Add(A=Multiply(A=Sine(None=@n83) {period=1.5}) {constB=3}) {constB=-1.5}
  n85 = AppendVector(A=Multiply(A=Saturate(None=@n84)) {constB=7}, B=Constant() {r=-1})
  n86 = Multiply(
   A=TextureSample(UVs=Add(A=@n76.UVs, B=@n80)) {texture=&T_UI_InspireStar}.R
   B=TextureSample(UVs=Add(A=@n81.UVs, B=@n85)) {texture=&T_UI_InspireStar}.R
  )
  n87 = MaterialFunctionCall(
   Texture Scale=@n52
   UVs=MaterialFunctionCall() {materialFunction=&MF_UI_UVScaling}.ScaleY
   Offset XY Texturesize Z=Constant3Vector() {constant=<rgba 0.127 0.04 0.0642 0>}
  ) {materialFunction=&ScaleUVsAroundPoint}
  n88 = Time()
  n89 = Multiply(A=@n88) {constB=0.15}
  n90 = Add(A=Multiply(A=Sine(None=@n89) {period=1.5}) {constB=3}) {constB=-1.5}
  n91 = AppendVector(A=Multiply(A=Saturate(None=@n90)) {constB=7}, B=Constant() {r=-1})
  n92 = Time()
  n93 = Sine(None=Multiply(A=@n92) {constB=0.15}) {period=1.5}
  n94 = Saturate(None=Add(A=Multiply(A=@n93) {constB=3}) {constB=-1.5})
  n95 = Add(
   A=MaterialFunctionCall(
    Texture Scale=@n52
    UVs=MaterialFunctionCall() {materialFunction=&MF_UI_UVScaling}.ScaleY
    Offset XY Texturesize Z=Constant3Vector() {constant=<rgba 0.107 0.04 0.0642 0>}
   ) {materialFunction=&ScaleUVsAroundPoint}.UVs
   B=AppendVector(A=Multiply(A=@n94) {constB=7}, B=Constant() {r=-1})
  )
  n96 = Time()
  n97 = Sine(None=Multiply(A=@n96) {constB=0.15}) {period=1.5}
  n98 = Saturate(None=Add(A=Multiply(A=@n97) {constB=3}) {constB=-1.5})
  n99 = Add(
   A=MaterialFunctionCall(
    Texture Scale=@n52
    UVs=MaterialFunctionCall() {materialFunction=&MF_UI_UVScaling}.ScaleY
    Offset XY Texturesize Z=Constant3Vector() {constant=<rgba 0.117 0.04 0.0642 0>}
   ) {materialFunction=&ScaleUVsAroundPoint}.UVs
   B=AppendVector(A=Multiply(A=@n98) {constB=7}, B=Constant() {r=-1})
  )
  n100 = Multiply(
   A=TextureSample(UVs=Add(A=@n87.UVs, B=@n91)) {texture=&T_UI_InspireStar}.R
   B=Multiply(
    A=TextureSample(UVs=@n95) {texture=&T_UI_InspireStar}.R
    B=TextureSample(UVs=@n99) {texture=&T_UI_InspireStar}.R
   )
  )
  LinearInterpolate(
   A=Constant3Vector() {constant=<rgba 0.599 0 0 0>}
   B=LinearInterpolate(
    A=Constant3Vector() {constant=<rgba 0.349 0.3121 0.271 0>}
    B=Constant3Vector() {constant=<rgba 0.7157 0.6654 0.5906 0>}
    Alpha=TextureSample(UVs=@n51) {texture=&T_UI_LightGrunge}.R
   )
   Alpha=Multiply(A=Multiply(A=@n61, B=@n75), B=Multiply(A=@n86, B=@n100))
  )
  MaterialFunctionCall!(In=@n1, Ramp Width=Constant() {r=8}) {materialFunction=/Engine/Functions/Engine_MaterialFunctions01/Math/SmoothCeil}
  Multiply(A=@n10.R, B=@n8.R)
  OneMinus(None=@n77)
  OneMinus(None=@n92)
  OneMinus(None=@n96)
  OneMinus(None=@n88)
  OneMinus(None=@n67)
  OneMinus(None=@n71)
  OneMinus(None=@n63)
  OneMinus(None=@n82)
  TextureSample(UVs=Rotator(Coordinate=@n3.UVs, Time=@n2)) {texture=&T_UI_InspireStar}
 ]
}
